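; Fire Petal Healthy Looter boon ability that throws ChillBronco (chug) projectiles at a player who searches a container.
; On activation it binds the avatar pawn's server searched-container event, storing the pawn and container.
; A Sequence with immediate and delayed branches staggers the shared spawn step.
; Projectiles launch from above the container, home on the searcher with low gravity, carry an explode effect spec and ignore collision with that container.
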(event Ability|EventOnEndAbility
  (params (bWasCancelled :type Boolean)))

(event Ability|EventActivateAbility
  (EventDispatchers|BindEventtoServer_OnSearchedContainer (CallFunction|GetAvatarasFortPlayerPawn) (delegate Custom|Server_OnSearchedContainerCustomEvent)))

(event Custom|Server_OnSearchedContainerCustomEvent
  (params (SearchingPawn :type Fort%20Pawn%20Object%20Reference) (Container :type Building%20Container%20Object%20Reference))
  (Variables|Default|SetSearching%20Pawn SearchingPawn)
  (Variables|Default|SetContainer Container)
  (Utilities|FlowControl|Sequence
    (:then_0
      _)
    (:then_1
      (Utilities|FlowControl|Delay 1.0))
    (:then_2
      (Utilities|FlowControl|Delay 3.0)))
  ; Spawn Chug Jug At Player
  (Variables|Default|SetChestToIgnore (Variables|Default|GetContainer))
  (bind _returnvalue (Projectile|SpawnProjectile :ProjectileClass "/FirePetalBoonsGameplay/Gameplay/HealthyLooter/B_Prj_Athena_ChillBroncoHealthyLooter.B_Prj_Athena_ChillBroncoHealthyLooter_C" :RequestedBy (Variables|Default|GetSearching%20Pawn) :SpawnLocation (+ (Transformation|GetActorLocation (Variables|Default|GetContainer)) "0.000000,0.000000,100.000000") :SpawnRotation "0, 0, 0" :SpawnDirection (Math|Conversions|RotationFromXVector (- (+ (Transformation|GetActorLocation (Variables|Default|GetSearching%20Pawn)) "0.000000,0.000000,-75.000000") (+ (Transformation|GetActorLocation (Variables|Default|GetContainer)) "0.000000,0.000000,100.000000"))) :EffectContainerSpecToApplyOnExplode (Ability|MakeGameplayEffectContainerSpec "(TagName=\"Abilities.generic.GenericProjectile.Explode\")") :bUseDefaultPhysics true :HomingTarget (Variables|Default|GetSearching%20Pawn) :InitialSpeed 5e+02 :GravityScale 0.1 :ChargePercent 0.0))
  (bind _asb_prj_athena_chill_bronco_healthy_looter (Utilities|Casting|CastToB_Prj_Athena_ChillBroncoHealthyLooter _returnvalue))
  (Variables|/FirePetalBoonsGameplay/Gameplay/HealthyLooter/B_Prj_Athena_ChillBroncoHealthyLooter.B_Prj_Athena_ChillBroncoHealthyLooter_C|SetIgnoreCollision (Variables|Default|GetContainer) _asb_prj_athena_chill_bronco_healthy_looter))
