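; Gameplay ability for the Healthy Looter boon that periodically spawns a ChillBronco healing splash projectile at the player pawn.
; Activation spawns one immediately and starts a looping five-second timer that repeats the spawn through SpawnChugSlash.
; Each spawn passes an explode effect container and applies the Water cooldown effect to the owner.
; The duplicated cooldown-tag checks have empty branches marked TODO, and EndAbility does not clear the timer.
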(event Custom|SpawnChugSlash
  (bind _returnvalue (CallFunction|GetAvatarasFortPlayerPawn))
  ; TODO: fix this
  (if (Ability|DoesTargetHaveTags _returnvalue (Variables|Default|GetCooldownTags))
    _)
  (Projectile|SpawnProjectile :ProjectileClass "/FirePetalBoonsGameplay/Gameplay/HealthyLooter/B_Prj_Athena_ChillBroncoHealthyLooter.B_Prj_Athena_ChillBroncoHealthyLooter_C" :RequestedBy _returnvalue :SpawnLocation (Transformation|GetActorLocation _returnvalue) :SpawnRotation "0, 0, 0" :SpawnDirection (Math|Rotator|MakeRotator 0.0) :EffectContainerSpecToApplyOnExplode (Ability|MakeGameplayEffectContainerSpec "(TagName=\"Abilities.generic.GenericProjectile.Explode\")") :bUseDefaultPhysics true :InitialSpeed 5e+02 :GravityScale 0.1 :ChargePercent 0.0)
  (Ability|ApplyGameplayEffectToOwner "/FirePetalBoonsGameplay/Gameplay/HealthyLooter/GE_HealthyLooter_Water_Cooldown.GE_HealthyLooter_Water_Cooldown_C"))

(event Ability|EventActivateAbility
  (bind _returnvalue (CallFunction|GetAvatarasFortPlayerPawn))
  (Utilities|FlowControl|Sequence
    (:then_0
      (Utilities|Time|SetTimerbyEvent (delegate Custom|SpawnChugSlash) 5.0 true))
    (:then_1
      ; TODO: fix this
      (if (Ability|DoesTargetHaveTags _returnvalue (Variables|Default|GetCooldownTags))
        _)
      (Projectile|SpawnProjectile :ProjectileClass "/FirePetalBoonsGameplay/Gameplay/HealthyLooter/B_Prj_Athena_ChillBroncoHealthyLooter.B_Prj_Athena_ChillBroncoHealthyLooter_C" :RequestedBy _returnvalue :SpawnLocation (Transformation|GetActorLocation _returnvalue) :SpawnRotation "0, 0, 0" :SpawnDirection (Math|Rotator|MakeRotator 0.0) :EffectContainerSpecToApplyOnExplode (Ability|MakeGameplayEffectContainerSpec "(TagName=\"Abilities.generic.GenericProjectile.Explode\")") :bUseDefaultPhysics true :InitialSpeed 5e+02 :GravityScale 0.1 :ChargePercent 0.0)
      (Ability|ApplyGameplayEffectToOwner "/FirePetalBoonsGameplay/Gameplay/HealthyLooter/GE_HealthyLooter_Water_Cooldown.GE_HealthyLooter_Water_Cooldown_C"))))

(event Ability|EventOnEndAbility
  (params (bWasCancelled :type Boolean)))
